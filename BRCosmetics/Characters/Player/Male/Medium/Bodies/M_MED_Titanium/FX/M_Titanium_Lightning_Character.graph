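# Unlit translucent, two-sided Niagara ribbon and mesh-particle material for lightning FX on the Titanium character.
# Stepped-time panning and center-based scaling build tapered UVs, while three offset lightning-texture samples create two-color chromatic aberration with glow.
# A camera-space texture sample drives a thresholded dissolve and a UV distortion that is multiplied by zero.
# Opacity combines a particle-age gradient mask, particle alpha and optional near-camera fade, with dynamic parameters controlling several effects.
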Material /BRCosmetics/Characters/Player/Male/Medium/Bodies/M_MED_Titanium/FX/M_Titanium_Lightning_Character
graph {
 properties {
  bUsedWithNiagaraMeshParticles=true
  bUsedWithNiagaraRibbons=true
  blendMode=BLEND_Translucent
  pixelDepthOffsetMode=PDOM_Legacy
  shadingModel=MSM_Unlit
  twoSided=true
  usedShadingModels=Unlit
 }
 nodes [
  n1 = TextureCoordinate() {coordinateIndex=1}
  Add(A=@n1, B=AppendVector(B=Constant()))
  DynamicParameter() {defaultValue=<rgba 0 0.005 0.5 0>, paramNames [Use_Camera_Fade, Color Aberration, Square Dissolve, UScale]}
  n2 = VectorParameter() {parameterName=Aberration Color 1, defaultValue=<rgba 3 0.3 0.3 1>}
  n3 = VectorParameter() {parameterName=Aberration Color 2, defaultValue=<rgba 0.15 3 1.29 1>}
  n4 = Sine(None=ComponentMask(None=TextureCoordinate()) {r=true})
  LinearInterpolate(A=Multiply(A=@n2.RGB), B=Multiply(A=@n3.RGB), Alpha=Saturate(None=Multiply(A=@n4) {constB=100}))
  n5 = TextureObjectParameter() {parameterName=Lightning Texture, texture=/BRCosmetics/Weapons/FORT_Melee/Pickaxe_Titanium/FX/Texture/T_Titanium_LightningV4}
  n6 = TextureSample(UVs=NamedRerouteUsage(), Tex=@n5) {samplerType=SAMPLERTYPE_LinearColor, texture=/BRCosmetics/Weapons/FORT_Melee/Pickaxe_Titanium/FX/Texture/T_Titanium_Lightning}
  n7 = DynamicParameter() {defaultValue=<rgba 0 0.005 0.5 0>, paramNames [Use_Camera_Fade, Color Aberration, Square Dissolve, UScale]}
  n8 = TextureSample(UVs=Add(A=NamedRerouteUsage(), B=@n7.Color Aberration), Tex=@n5) {samplerType=SAMPLERTYPE_LinearColor, texture=&T_Titanium_Lightning}
  n9 = TextureSample(UVs=Add(A=NamedRerouteUsage(), B=Multiply(A=@n7.Color Aberration) {constB=-1}), Tex=@n5) {samplerType=SAMPLERTYPE_LinearColor, texture=&T_Titanium_Lightning}
  n10 = Max(A=Max(A=@n6.R, B=@n8.R), B=@n9.R)
  n11 = ParticleColor()
  n12 = DynamicParameter() {defaultValue=<rgba 0 0.005 0.5 0>, paramNames [Use_Camera_Fade, Color Aberration, Square Dissolve, UScale]}
  n13 = Multiply(A=Add(A=Multiply(A=@n8.R, B=@n2.RGB), B=Multiply(A=@n9.R, B=@n3.RGB)), B=@n12.Color Aberration)
  n14 = Subtract(A=Multiply(A=@n11.RGB, B=@n10), B=@n13)
  n15 = Multiply(
   A=LinearInterpolate(A=VectorParameter() {parameterName=Glow Color, defaultValue=<rgba 35 35 35 1>}.RGB, Alpha=@n10)
   B=Add(A=Multiply(A=@n14, B=@n6.R), B=@n13)
  )
  MaterialFunctionCall(Emissive=@n15) {materialFunction=&MF_FX_EmissiveAdjustments}
  n16 = Time()
  n17 = ScalarParameter() {defaultValue=1, parameterName=Speed X, sliderMax=10}
  n18 = ScalarParameter() {defaultValue=10, parameterName=Steps, sliderMax=10}
  n19 = Floor(None=Multiply(A=Multiply(A=@n16, B=@n17), B=@n18))
  n20 = Frac(None=AppendVector(A=Divide(A=@n19, B=@n18), B=Constant()))
  n21 = DynamicParameter() {defaultValue=<rgba 0 0.005 0.5 0>, paramNames [Use_Camera_Fade, Color Aberration, Square Dissolve, UScale]}
  n22 = ScalarParameter() {defaultValue=1, parameterName=Scale V, sliderMax=5}
  n23 = SphereMask(A=ComponentMask(None=TextureCoordinate()) {r=true}, B=Constant() {r=0.5}) {attenuationRadius=0.51, hardnessPercent=10}
  n24 = MaterialFunctionCall(
   UVs=Multiply(A=Add(A=@n1, B=@n20), B=AppendVector(A=@n21.UScale, B=@n22))
   Texture Scale=AppendVector(A=Constant() {r=1}, B=Saturate(None=@n23))
  ) {materialFunction=&ScaleUVsByCenter}
  n25 = ComponentMask(None=@n24.UVs) {g=true}
  NamedRerouteDeclaration(None=AppendVector(A=ComponentMask(None=@n24.UVs) {r=true}, B=Saturate(None=@n25))) {name=UV, nodeColor=<rgba 0 0.1797 1 1>}
  n26 = VectorParameter() {parameterName=Distortion Vector, defaultValue=<rgba 0 1 0 1>}
  n27 = WorldPosition()
  n28 = Multiply(
   A=ComponentMask(None=TransformPosition(None=@n27.XYZ) {transformSourceType=TRANSFORMPOSSOURCE_World, transformType=TRANSFORMPOSSOURCE_Camera}) {g=true}
   B=ScalarParameter() {defaultValue=0.1, parameterName=Distortion Tiling, sliderMax=5}
  )
  n29 = Constant() {r=0.5}
  n30 = TextureSample(
   UVs=Panner(Coordinate=AppendVector(A=@n28, B=@n29))
   Tex=TextureObjectParameter() {parameterName=Lightning Texture, texture=&T_Titanium_LightningV4}
  ) {texture=&T_Titanium_LightningV4}
  n31 = ComponentMask(None=Multiply(A=Transform(None=@n26.RGB) {transformSourceType=TRANSFORMSOURCE_View, transformType=TRANSFORM_Tangent}, B=@n30.B)) {g=true, r=true}
  n32 = Constant()
  NamedRerouteDeclaration(None=Add(A=NamedRerouteUsage(), B=Multiply(A=@n31, B=@n32))) {name=Distorted UV, nodeColor=<rgba 1 0.75 0 1>}
  n33 = DynamicParameter() {defaultValue=<rgba 0 0.005 0.5 0>, paramNames [Use_Camera_Fade, Color Aberration, Square Dissolve, UScale]}
  NamedRerouteDeclaration(None=If(A=@n30.B, B=@n33.Square Dissolve, "A > B"=Constant() {r=1}, "A < B"=Constant())) {name=Dissolve, nodeColor=<rgba 0 1 0.3672 1>}
  n34 = ComponentMask(None=TextureCoordinate()) {r=true}
  n35 = Saturate(None=Power(Base=OneMinus(None=@n34)) {constExponent=1.3})
  n36 = Frac(None=ParticleRelativeTime())
  n37 = ScalarParameter() {defaultValue=1, parameterName=LightningSection}
  n38 = Add(A=@n6.G, B=@n10)
  n39 = Multiply(
   A=Frac(None=MaterialFunctionCall!(Gradient=@n35, Mask Position=@n36, Mask Size=@n37) {materialFunction=/Game/Effects/Fort_Effects/MaterialFunctions/MF_GradientMask}.Linear)
   B=Multiply(A=Saturate(None=@n38), B=NamedRerouteUsage())
  ) {constA=1}
  n40 = MaterialFunctionCall(Distance=ScalarParameter() {defaultValue=870, parameterName=CameraDistance_Fade}) {materialFunction=&Near_Camera_Fade}
  n41 = Multiply(A=@n39, B=Saturate(None=Multiply(A=@n40.Result) {constB=5}))
  n42 = DynamicParameter() {defaultValue=<rgba 0 0.005 0.5 0>, paramNames [Use_Camera_Fade, Color Aberration, Square Dissolve, UScale]}
  Saturate(None=Multiply(A=@n11.A, B=LinearInterpolate(A=@n39, B=@n41, Alpha=@n42.Use_Camera_Fade)))
  ScalarParameter() {defaultValue=0.7, parameterName=Scale U, sliderMax=5}
  StaticSwitchParameter(True=@n11.RGB, False=@n15) {parameterName=Glow}
  StaticSwitchParameter(True=@n6.G, False=@n10) {parameterName=Glow}
 ]
}
