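# Niagara idle effect for the Titanium character body that crackles lightning on both sides using meshes, sprites and lights.
# Mirrored left/right emitters follow per-side skeletal mesh sockets, with a reduced socket set when User.bisHandHidden is true.
# One system timeline ramps the regular arcs' spawn probability, size, lifetime and brightness, while a second drives spawn-rate spikes for the burst emitters.
# Random visibility tags vary the particles, and front-end modifiers adjust material parameters.
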NiagaraSystem /BRCosmetics/Characters/Player/Male/Medium/Bodies/M_MED_Titanium/FX/NS_Titanium_Character_Idle
dependencies {
 usedDataInterfaces [
  &NiagaraDataInterfaceArrayFloat
  &NiagaraDataInterfaceArrayInt32
  &NiagaraDataInterfaceCurve
  &NiagaraDataInterfaceColorCurve
  &NiagaraDataInterfaceSkeletalMesh
 ]
 usedDynamicInputs [&FNTimeOfDayScalar]
 usedModules [
  &self:SystemSpawnScript.SystemScriptSource.SystemScriptGraph.NiagaraNodeAssignment_0.SetVariables_#1
  &SystemState
  &self:SystemSpawnScript.SystemScriptSource.SystemScriptGraph.NiagaraNodeAssignment_2.SetVariables_#2
  &Timeline
 ]
}
emitters {
 Lightning_Burst_Left {
  emitterUpdateScript [
   &EmitterState(Life Cycle Mode=::NewEnumerator1, Loop Duration=1, UseLoopDelay=false)
   &SpawnRate(
    SpawnRate=&Lerp_Float(
     A=0
     B=30
     Alpha=&FloatFromCurve(FloatCurve=DataInterface() {propertyValues {Curve {keys [<at 0 0>,<at 0.2 1>,<at 1 0>]}}}, CurveIndex=$System.Timeline001.NormalizedTime)
    )
    Use Spawn Probability=false
   )
  ]
  particleSpawnScript [
   &InitializeParticle(
    Color=&RandomRangeLinearColor(Minimum=<rgba 0.7 0.7 0.7 1>, Maximum=<rgba 1.3 1.3 1.3 1>)
    Lifetime=&RandomRangeFloat(Minimum=0.2, Maximum=0.3)
    Mesh Scale=&RandomRangeVector(Minimum=<0.8 0.8 0.8>, Maximum=<1.3 2.5 2.5>)
    Mesh Scale Mode=::NewEnumerator4
    Sprite Rotation Mode=::NewEnumerator2
    Sprite Size Mode=::NewEnumerator3
    Uniform Sprite Size=45
    UsePositionOffset=false
    Write Material Random=true
   )
   &CameraOffset(Camera Offset Amount=8)
   /SetVariables() {isSetParameters=true, module=SetVariables_#3}
   /SetVariables() {isSetParameters=true, module=SetVariables_#4}
   &InitialMeshOrientation(
    Mesh Axis To Orient=<0 1 0>
    Mesh Orientation Mode=::NewEnumerator2
    Orientation Vector=&RandomRangeVector(Minimum=<0 0 0>, Maximum=<0 1 0>)
    Use Rotation Vector=false
   ) {enabled=false}
   /SetVariables(Particles.VisibilityTag=&RandomRangeInt(Minimum=0, Maximum=1)) {isSetParameters=true, module=SetVariables_#5}
   &DynamicMaterialParameters(
    Index 0 Param 2=&RandomRangeFloat(Minimum=0.2, Maximum=0.4)
    Index 1 Param 2=&FN_FrontEndFloatModifier(Gameplay=1, Front End=0)
    Module.Param0WriteEnabled=false
    Module.Param0WriteEnabled1=false
    Module.Param1WriteEnabled=true
    Module.Param1WriteEnabled1=true
    Module.Param2WriteEnabled=false
    Module.Param2WriteEnabled1=false
    Module.Param3WriteEnabled=false
    Module.Param3WriteEnabled1=false
    Write Parameter Index 1=true
   ) {module=DynamicMaterialParameters001}
  ]
  particleUpdateScript [
   &ParticleState()
   &SkeletalMeshLocation(
    Bone / Socket Index=&MakeCustomIntFromBool(
     Bool=$User.bisHandHidden
     True INT=&SelectIntFromArray(
      Array Sampling Mode=::NewEnumerator1
      Direct Array Index=?
      Int Selection Array=DataInterface() {propertyValues {IntData [1,5]}}
      Random Seed=0
     )
     False INT=&SelectIntFromArray(
      Array Sampling Mode=::NewEnumerator1
      Direct Array Index=?
      Int Selection Array=DataInterface() {propertyValues {IntData [1,3,5]}}
      Random Seed=0
     )
    )
    MaskSpawnGroup=false
    Mesh Sampling Type=::NewEnumerator5
    Orientation Sampling=::NewEnumerator1
    Skeletal Mesh=$System.Skeletal Mesh
    Socket Sampling Mode=::NewEnumerator5
   )
   /SetVariables(Particles.PositionGlowBig:NiagaraPosition=&AddVectorToPosition(Position=$Particles.SkeletalMeshLocation.SampledPosition, Vector=<0 0 15>)) {isSetParameters=true, module=SetVariables_#6}
   &ScaleSpriteSize(Uniform Curve Sprite Scale=DataInterface() {propertyValues {Curve {keys [<at 0 0.3 RCIM_Cubic>,<at 0.1 1 RCIM_Cubic>]}}})
   &ScaleMeshSize(Scale Factor=&MakeVector(
     X=&FloatFromCurve(FloatCurve=DataInterface() {propertyValues {Curve {keys [
          {arriveTangent=1.0254, interpMode=RCIM_Cubic, leaveTangent=1.0254, tangentMode=RCTM_User, time=0, value=0.7}
          {interpMode=RCIM_Cubic, tangentMode=RCTM_User, time=0.5, value=1}
         ]}}})
     Y=&FloatFromCurve(FloatCurve=DataInterface() {propertyValues {Curve {keys [
          {arriveTangent=1.0254, interpMode=RCIM_Cubic, leaveTangent=1.0254, tangentMode=RCTM_User, time=0, value=0.7}
          {interpMode=RCIM_Cubic, tangentMode=RCTM_User, time=0.5, value=1}
         ]}}})
     Z=&FloatFromCurve(FloatCurve=DataInterface() {propertyValues {Curve {keys [
          {arriveTangent=0.8974, interpMode=RCIM_Cubic, leaveTangent=0.8974, tangentMode=RCTM_User, time=0, value=0.5}
          {interpMode=RCIM_Cubic, tangentMode=RCTM_User, time=0.75, value=1}
         ]}}})
    )) {enabled=false}
   &UpdateMeshOrientation(
    Facing Direction=&RotateVectorByQuaternion(
     Vector To Rotate=&RandomRangeVector(Minimum=<0.8 0.8 0.7>, Maximum=<1 1 1>)
     Quaternion=$Particles.SkeletalMeshLocation.SampledRotation
    )
    Facing Interpolation=false
    Orientation Method=::NewEnumerator1
    Override Quat=false
    TransformFacing=false
    TransformUp=false
   ) {module=UpdateMeshOrientation001}
   /SetVariables(Particles.ColorSprite:LinearColor=&MakeLinearColorFromVectorAndFloat(
     "Vector (RGB)"=&MakeVectorFromLinearColorRGB(Color=&ColorFromCurve(ColorCurve=DataInterface() {propertyValues {
         AlphaCurve {keys []}
         BlueCurve {keys [<at 0.0908 30>,<at 0.1816 1 RCIM_Cubic>]}
         GreenCurve {keys [<at 0.0908 30>,<at 0.1816 1 RCIM_Cubic>]}
         RedCurve {keys [<at 0.0908 30>,<at 0.1816 1 RCIM_Cubic>]}
        }}))
     "Float (Alpha)"=&FloatFromCurve(
      FloatCurve=DataInterface() {propertyValues {Curve {keys [
          {arriveTangent=0.0039, interpMode=RCIM_Cubic, leaveTangent=0.0039, tangentMode=RCTM_User, time=0.1, value=0}
          {arriveTangent=0.007, interpMode=RCIM_Cubic, leaveTangent=0.007, tangentMode=RCTM_User, time=0.2, value=0.6}
          {arriveTangent=-0.0192, interpMode=RCIM_Cubic, leaveTangent=-0.0192, tangentMode=RCTM_User, time=0.3, value=0.6}
          <at 0.8 0 RCIM_Cubic>
         ]}}}
      Scale Curve=0.1
     )
    )) {isSetParameters=true, module=SetVariables_#7}
   &ScaleColor(
    Scale Alpha=&FloatFromCurve(FloatCurve=DataInterface() {propertyValues {Curve {keys [<at 0.8 2>,<at 1 0>]}}})
    Scale RGB=&MakeVectorFromLinearColorRGB(Color=&ColorFromCurve(ColorCurve=DataInterface() {propertyValues {
        AlphaCurve {keys []}
        BlueCurve {keys [<at 0 3 RCIM_Cubic>,<at 0.2659 0.2>]}
        GreenCurve {keys [<at 0 3 RCIM_Cubic>,<at 0.2659 0.2>]}
        RedCurve {keys [<at 0 3 RCIM_Cubic>,<at 0.2659 0.2>]}
       }}))
    ScaleA=true
    ScaleRGB=true
   )
   &DynamicMaterialParameters(
    Index 0 Param 2=0.2
    Index 0 Param 3=8
    Index 0 Param 4=&FloatFromCurve(
     FloatCurve=DataInterface() {propertyValues {Curve {keys [
         {arriveTangent=0.1537, interpMode=RCIM_Cubic, leaveTangent=0.1537, tangentMode=RCTM_User, time=0.6, value=0}
         {arriveTangent=4.1515, interpMode=RCIM_Cubic, leaveTangent=4.1515, tangentMode=RCTM_User, time=1, value=1.525}
        ]}}}
     Scale Curve=0.2
    )
    Index 1 Param 1=&FloatFromCurve(
     FloatCurve=DataInterface() {propertyValues {Curve {keys [
         {arriveTangent=0.1804, interpMode=RCIM_Cubic, leaveTangent=0.1804, tangentMode=RCTM_User, time=0.5, value=0}
         {arriveTangent=3.1151, interpMode=RCIM_Cubic, leaveTangent=3.1151, tangentMode=RCTM_User, time=0.9, value=0.3}
        ]}}}
     Scale Curve=0.03
    )
    Module.Param0WriteEnabled=false
    Module.Param0WriteEnabled1=true
    Module.Param1WriteEnabled=false
    Module.Param1WriteEnabled1=false
    Module.Param2WriteEnabled=true
    Module.Param2WriteEnabled1=false
    Module.Param3WriteEnabled=true
    Module.Param3WriteEnabled1=false
    Write Parameter Index 1=true
   )
  ]
  rendererClasses [
   &NiagaraSpriteRendererProperties
   &NiagaraLightRendererProperties
   &NiagaraMeshRendererProperties
  ]
  renderers [
   {rendererClass=&NiagaraSpriteRendererProperties}
   {rendererClass=&NiagaraLightRendererProperties, rendererIndex=1}
   {rendererClass=&NiagaraMeshRendererProperties, rendererIndex=2}
   {rendererClass=&NiagaraMeshRendererProperties, rendererIndex=3}
   {rendererClass=&NiagaraMeshRendererProperties, rendererIndex=4}
  ]
 }
 Lightning_Burst_Right {
  emitterUpdateScript [
   &EmitterState(Life Cycle Mode=::NewEnumerator1, Loop Duration=1, UseLoopDelay=false)
   &SpawnRate(
    SpawnRate=&Lerp_Float(
     A=0
     B=30
     Alpha=&FloatFromCurve(FloatCurve=DataInterface() {propertyValues {Curve {keys [<at 0 0>,<at 0.2 1>,<at 1 0>]}}}, CurveIndex=$System.Timeline001.NormalizedTime)
    )
    Use Spawn Probability=false
   )
  ]
  particleSpawnScript [
   &InitializeParticle(
    Color=&RandomRangeLinearColor(Minimum=<rgba 0.7 0.7 0.7 1>, Maximum=<rgba 1.3 1.3 1.3 1>)
    Lifetime=&RandomRangeFloat(Minimum=0.2, Maximum=0.3)
    Mesh Scale=&RandomRangeVector(Minimum=<0.8 0.8 0.8>, Maximum=<1.3 2.5 2.5>)
    Mesh Scale Mode=::NewEnumerator4
    Sprite Rotation Mode=::NewEnumerator2
    Sprite Size Mode=::NewEnumerator3
    Uniform Sprite Size=45
    UsePositionOffset=false
    Write Material Random=true
   )
   &CameraOffset(Camera Offset Amount=15)
   /SetVariables() {isSetParameters=true, module=SetVariables_#3}
   /SetVariables() {isSetParameters=true, module=SetVariables_#4}
   &InitialMeshOrientation(
    Mesh Axis To Orient=<0 1 0>
    Mesh Orientation Mode=::NewEnumerator2
    Orientation Vector=&RandomRangeVector(Minimum=<0 0 0>, Maximum=<0 1 0>)
    Use Rotation Vector=false
   ) {enabled=false}
   /SetVariables(Particles.VisibilityTag=&RandomRangeInt(Minimum=0, Maximum=1)) {isSetParameters=true, module=SetVariables_#5}
   &DynamicMaterialParameters(
    Index 0 Param 2=&RandomRangeFloat(Minimum=0.2, Maximum=0.4)
    Index 1 Param 2=&FN_FrontEndFloatModifier(Gameplay=1, Front End=0)
    Module.Param0WriteEnabled=false
    Module.Param0WriteEnabled1=false
    Module.Param1WriteEnabled=true
    Module.Param1WriteEnabled1=true
    Module.Param2WriteEnabled=false
    Module.Param2WriteEnabled1=false
    Module.Param3WriteEnabled=false
    Module.Param3WriteEnabled1=false
    Write Parameter Index 1=true
   ) {module=DynamicMaterialParameters001}
  ]
  particleUpdateScript [
   &ParticleState()
   &SkeletalMeshLocation(
    Bone / Socket Index=&MakeCustomIntFromBool(
     Bool=$User.bisHandHidden
     True INT=&SelectIntFromArray(
      Array Sampling Mode=::NewEnumerator1
      Direct Array Index=?
      Int Selection Array=DataInterface() {propertyValues {IntData [0,4]}}
      Random Seed=0
     )
     False INT=&SelectIntFromArray(
      Array Sampling Mode=::NewEnumerator1
      Direct Array Index=?
      Int Selection Array=DataInterface() {propertyValues {IntData [0,2,4]}}
      Random Seed=0
     )
    )
    MaskSpawnGroup=false
    Mesh Sampling Type=::NewEnumerator5
    Orientation Sampling=::NewEnumerator1
    Skeletal Mesh=$System.Skeletal Mesh
    Socket Sampling Mode=::NewEnumerator5
   )
   /SetVariables(Particles.PositionGlowBig:NiagaraPosition=&AddVectorToPosition(Position=$Particles.SkeletalMeshLocation.SampledPosition, Vector=<0 0 15>)) {isSetParameters=true, module=SetVariables_#6}
   &ScaleSpriteSize(Uniform Curve Sprite Scale=DataInterface() {propertyValues {Curve {keys [<at 0 0.6 RCIM_Cubic>,<at 0.1 1 RCIM_Cubic>]}}})
   &ScaleMeshSize(Scale Factor=&MakeVector(
     X=&FloatFromCurve(FloatCurve=DataInterface() {propertyValues {Curve {keys [
          {arriveTangent=1.0254, interpMode=RCIM_Cubic, leaveTangent=1.0254, tangentMode=RCTM_User, time=0, value=0.7}
          {interpMode=RCIM_Cubic, tangentMode=RCTM_User, time=0.5, value=1}
         ]}}})
     Y=&FloatFromCurve(FloatCurve=DataInterface() {propertyValues {Curve {keys [
          {arriveTangent=1.0254, interpMode=RCIM_Cubic, leaveTangent=1.0254, tangentMode=RCTM_User, time=0, value=0.7}
          {interpMode=RCIM_Cubic, tangentMode=RCTM_User, time=0.5, value=1}
         ]}}})
     Z=&FloatFromCurve(FloatCurve=DataInterface() {propertyValues {Curve {keys [
          {arriveTangent=0.8974, interpMode=RCIM_Cubic, leaveTangent=0.8974, tangentMode=RCTM_User, time=0, value=0.5}
          {interpMode=RCIM_Cubic, tangentMode=RCTM_User, time=0.75, value=1}
         ]}}})
    )) {enabled=false}
   &UpdateMeshOrientation(
    Facing Direction=&RotateVectorByQuaternion(
     Vector To Rotate=&RandomRangeVector(Minimum=<0.8 0.8 0.7>, Maximum=<1 1 1>)
     Quaternion=$Particles.SkeletalMeshLocation.SampledRotation
    )
    Facing Interpolation=false
    Orientation Method=::NewEnumerator1
    Override Quat=false
    TransformFacing=false
    TransformUp=false
   ) {module=UpdateMeshOrientation001}
   /SetVariables(Particles.ColorSprite:LinearColor=&MakeLinearColorFromVectorAndFloat(
     "Vector (RGB)"=&MakeVectorFromLinearColorRGB(Color=&ColorFromCurve(ColorCurve=DataInterface() {propertyValues {
         AlphaCurve {keys []}
         BlueCurve {keys [<at 0.0908 30>,<at 0.1816 1 RCIM_Cubic>]}
         GreenCurve {keys [<at 0.0908 30>,<at 0.1816 1 RCIM_Cubic>]}
         RedCurve {keys [<at 0.0908 30>,<at 0.1816 1 RCIM_Cubic>]}
        }}))
     "Float (Alpha)"=&FloatFromCurve(
      FloatCurve=DataInterface() {propertyValues {Curve {keys [
          {arriveTangent=0.0039, interpMode=RCIM_Cubic, leaveTangent=0.0039, tangentMode=RCTM_User, time=0.1, value=0}
          {arriveTangent=0.007, interpMode=RCIM_Cubic, leaveTangent=0.007, tangentMode=RCTM_User, time=0.2, value=0.6}
          {arriveTangent=-0.0192, interpMode=RCIM_Cubic, leaveTangent=-0.0192, tangentMode=RCTM_User, time=0.3, value=0.6}
          <at 0.8 0 RCIM_Cubic>
         ]}}}
      Scale Curve=0.1
     )
    )) {isSetParameters=true, module=SetVariables_#7}
   &ScaleColor(
    Scale Alpha=&FloatFromCurve(FloatCurve=DataInterface() {propertyValues {Curve {keys [<at 0.8 2>,<at 1 0>]}}})
    Scale RGB=&MakeVectorFromLinearColorRGB(Color=&ColorFromCurve(ColorCurve=DataInterface() {propertyValues {
        AlphaCurve {keys []}
        BlueCurve {keys [<at 0 3 RCIM_Cubic>,<at 0.2659 0.2>]}
        GreenCurve {keys [<at 0 3 RCIM_Cubic>,<at 0.2659 0.2>]}
        RedCurve {keys [<at 0 3 RCIM_Cubic>,<at 0.2659 0.2>]}
       }}))
    ScaleA=true
    ScaleRGB=true
   )
   &DynamicMaterialParameters(
    Index 0 Param 2=0.2
    Index 0 Param 3=8
    Index 0 Param 4=&FloatFromCurve(
     FloatCurve=DataInterface() {propertyValues {Curve {keys [
         {arriveTangent=0.1537, interpMode=RCIM_Cubic, leaveTangent=0.1537, tangentMode=RCTM_User, time=0.6, value=0}
         {arriveTangent=4.1515, interpMode=RCIM_Cubic, leaveTangent=4.1515, tangentMode=RCTM_User, time=1, value=1.525}
        ]}}}
     Scale Curve=0.2
    )
    Index 1 Param 1=&FloatFromCurve(
     FloatCurve=DataInterface() {propertyValues {Curve {keys [
         {arriveTangent=0.1804, interpMode=RCIM_Cubic, leaveTangent=0.1804, tangentMode=RCTM_User, time=0.5, value=0}
         {arriveTangent=3.1151, interpMode=RCIM_Cubic, leaveTangent=3.1151, tangentMode=RCTM_User, time=0.9, value=0.3}
        ]}}}
     Scale Curve=0.03
    )
    Module.Param0WriteEnabled=false
    Module.Param0WriteEnabled1=true
    Module.Param1WriteEnabled=false
    Module.Param1WriteEnabled1=false
    Module.Param2WriteEnabled=true
    Module.Param2WriteEnabled1=false
    Module.Param3WriteEnabled=true
    Module.Param3WriteEnabled1=false
    Write Parameter Index 1=true
   )
  ]
  rendererClasses [
   &NiagaraSpriteRendererProperties
   &NiagaraLightRendererProperties
   &NiagaraMeshRendererProperties
  ]
  renderers [
   {rendererClass=&NiagaraSpriteRendererProperties}
   {rendererClass=&NiagaraLightRendererProperties, rendererIndex=1}
   {rendererClass=&NiagaraMeshRendererProperties, rendererIndex=2}
   {rendererClass=&NiagaraMeshRendererProperties, rendererIndex=3}
   {rendererClass=&NiagaraMeshRendererProperties, rendererIndex=4}
  ]
 }
 Lightning_Left {
  emitterUpdateScript [
   &EmitterState(Life Cycle Mode=::NewEnumerator1, Loop Duration=1, UseLoopDelay=false)
   &SpawnRate(
    Spawn Probability=&Lerp_Float(A=0.15, B=0.8, Alpha=$System.Timeline.NormalizedTime)
    SpawnRate=10
    Use Spawn Probability=true
   )
  ]
  particleSpawnScript [
   &InitializeParticle(
    Color=&RandomRangeLinearColor(Minimum=<rgba 0.7 0.7 0.7 1>, Maximum=<rgba 1.3 1.3 1.3 1>)
    Lifetime=&Lerp_Float(
     A=&RandomRangeFloat(Minimum=0.3, Maximum=0.4)
     B=&RandomRangeFloat(Minimum=0.16, Maximum=0.3)
     Alpha=$System.Timeline.NormalizedTime
    )
    Mesh Scale=&Lerp_Vector(
     A=&RandomRangeVector(Minimum=<0.8 0.8 0.8>, Maximum=<1.5 1.5 1.3>)
     B=&RandomRangeVector(Minimum=<1 1 1>, Maximum=<2 2 2>)
     Alpha=$System.Timeline.NormalizedTime
    )
    Mesh Scale Mode=::NewEnumerator4
    Sprite Rotation Mode=::NewEnumerator2
    Sprite Size Mode=::NewEnumerator3
    Uniform Sprite Size=15
    UsePositionOffset=false
    Write Material Random=true
   )
   &CameraOffset(Camera Offset Amount=5)
   /SetVariables(Particles.OffsetMesh=4) {isSetParameters=true, module=SetVariables_#8}
   /SetVariables() {isSetParameters=true, module=SetVariables_#3}
   /SetVariables(Particles.VisibilityTag=&RandomRangeInt(Minimum=0, Maximum=2)) {isSetParameters=true}
   /SetVariables() {isSetParameters=true}
   &DynamicMaterialParameters(
    Index 0 Param 1=&FN_FrontEndFloatModifier(Gameplay=1, Front End=0)
    Index 0 Param 4=&SelectFloatFromArray(Direct Array Index=$Particles.VisibilityTag, Float Selection Array=DataInterface() {propertyValues {FloatData [0.5,1.5,1.5]}})
    Module.Param0WriteEnabled=true
    Module.Param1WriteEnabled=false
    Module.Param2WriteEnabled=false
    Module.Param3WriteEnabled=true
   )
  ]
  particleUpdateScript [
   &ParticleState()
   &SkeletalMeshLocation(
    Bone / Socket Index=&MakeCustomIntFromBool(
     Bool=$User.bisHandHidden
     True INT=&SelectIntFromArray(
      Array Sampling Mode=::NewEnumerator1
      Direct Array Index=?
      Int Selection Array=DataInterface() {propertyValues {IntData [1,5]}}
      Random Seed=0
     )
     False INT=&SelectIntFromArray(
      Array Sampling Mode=::NewEnumerator1
      Direct Array Index=?
      Int Selection Array=DataInterface() {propertyValues {IntData [1,3,5]}}
      Random Seed=0
     )
    )
    MaskSpawnGroup=false
    Mesh Sampling Type=::NewEnumerator5
    Orientation Sampling=::NewEnumerator1
    Skeletal Mesh=$System.Skeletal Mesh
    Socket Sampling Mode=::NewEnumerator5
   )
   &ScaleSpriteSize(Uniform Curve Sprite Scale=DataInterface() {propertyValues {Curve {keys [<at 0 0.3 RCIM_Cubic>,<at 0.2 1 RCIM_Cubic>]}}})
   &ScaleMeshSize(Scale Factor=&MakeVector(
     X=&FloatFromCurve(FloatCurve=DataInterface() {propertyValues {Curve {keys [
          {arriveTangent=1.0254, interpMode=RCIM_Cubic, leaveTangent=1.0254, tangentMode=RCTM_User, time=0, value=0.7}
          {interpMode=RCIM_Cubic, tangentMode=RCTM_User, time=0.5, value=1}
         ]}}})
     Y=&FloatFromCurve(FloatCurve=DataInterface() {propertyValues {Curve {keys [
          {arriveTangent=1.0254, interpMode=RCIM_Cubic, leaveTangent=1.0254, tangentMode=RCTM_User, time=0, value=0.7}
          {interpMode=RCIM_Cubic, tangentMode=RCTM_User, time=0.5, value=1}
         ]}}})
     Z=&FloatFromCurve(FloatCurve=DataInterface() {propertyValues {Curve {keys [
          {arriveTangent=0.8974, interpMode=RCIM_Cubic, leaveTangent=0.8974, tangentMode=RCTM_User, time=0, value=0.5}
          {interpMode=RCIM_Cubic, tangentMode=RCTM_User, time=0.75, value=1}
         ]}}})
    ))
   &UpdateMeshOrientation(
    Facing Direction=&RotateVectorByQuaternion(
     Vector To Rotate=&RandomRangeVector(Minimum=<0.8 0.8 0.35>, Maximum=<1 1 0.5>)
     Quaternion=$Particles.SkeletalMeshLocation.SampledRotation
    )
    Facing Interpolation=false
    Orientation Method=::NewEnumerator1
    Override Quat=false
    TransformFacing=false
    TransformUp=false
   ) {module=UpdateMeshOrientation001}
   &ScaleColor(
    Scale Alpha=&FloatFromCurve(FloatCurve=DataInterface() {propertyValues {Curve {keys [
         {arriveTangent=14.472, interpMode=RCIM_Cubic, leaveTangent=14.472, tangentMode=RCTM_User, time=0.12, value=0.5}
         {arriveTangent=-0.0363, interpMode=RCIM_Cubic, leaveTangent=-0.0363, tangentMode=RCTM_User, time=0.17, value=1}
         <at 0.9 1>
         <at 0.95 0>
        ]}}})
    Scale RGB=&Lerp_Vector(
     A=&MakeVectorFromLinearColorRGB(Color=&ColorFromCurve(ColorCurve=DataInterface() {propertyValues {
         AlphaCurve {keys []}
         BlueCurve {keys [<at 0.1007 6 RCIM_Cubic>,<at 0.1964 0.1>,<at 0.4217 0.2>,<at 0.6204 2>]}
         GreenCurve {keys [<at 0.1007 6 RCIM_Cubic>,<at 0.1964 0.1>,<at 0.4217 0.2>,<at 0.6204 2>]}
         RedCurve {keys [<at 0.1007 6 RCIM_Cubic>,<at 0.1964 0.1>,<at 0.4217 0.2>,<at 0.6204 2>]}
        }}))
     B=&MakeVectorFromLinearColorRGB(Color=&ColorFromCurve(ColorCurve=DataInterface() {propertyValues {
         AlphaCurve {keys []}
         BlueCurve {keys [{arriveTangent=1.5484, interpMode=RCIM_Cubic, time=0.0816, value=14}, <at 0.1893 0.2>, <at 0.3809 0.3>, <at 0.5904 2>]}
         GreenCurve {keys [{arriveTangent=1.5484, interpMode=RCIM_Cubic, time=0.0816, value=14}, <at 0.1893 0.2>, <at 0.3809 0.3>, <at 0.5904 2>]}
         RedCurve {keys [{arriveTangent=1.5484, interpMode=RCIM_Cubic, time=0.0816, value=14}, <at 0.1893 0.2>, <at 0.3809 0.3>, <at 0.5904 2>]}
        }}))
     Alpha=$System.Timeline.NormalizedTime
    )
    ScaleA=true
    ScaleRGB=true
   ) {enabled=false}
   &ScaleColor(
    Scale Alpha=&FloatFromCurve(
     FloatCurve=DataInterface() {propertyValues {Curve {keys [
         {arriveTangent=14.472, interpMode=RCIM_Cubic, leaveTangent=14.472, tangentMode=RCTM_User, time=0.12, value=0.5}
         {arriveTangent=-0.0363, interpMode=RCIM_Cubic, leaveTangent=-0.0363, tangentMode=RCTM_User, time=0.17, value=1}
         <at 0.95 1>
         <at 1 0>
        ]}}}
     Scale Curve=3
    )
    Scale RGB=&Lerp_Vector(
     A=&MakeVectorFromLinearColorRGB(Color=&ColorFromCurve(ColorCurve=DataInterface() {propertyValues {
         AlphaCurve {keys []}
         BlueCurve {keys [<at 0.0146 1 RCIM_Cubic>,<at 0.1743 0.07>]}
         GreenCurve {keys [<at 0.0146 1 RCIM_Cubic>,<at 0.1743 0.07>]}
         RedCurve {keys [<at 0.0146 1 RCIM_Cubic>,<at 0.1743 0.07>]}
        }}))
     B=&MakeVectorFromLinearColorRGB(Color=&ColorFromCurve(ColorCurve=DataInterface() {propertyValues {
         AlphaCurve {keys []}
         BlueCurve {keys [{arriveTangent=1.5484, interpMode=RCIM_Cubic, time=0, value=5}, <at 0.1898 0.2>]}
         GreenCurve {keys [{arriveTangent=1.5484, interpMode=RCIM_Cubic, time=0, value=5}, <at 0.1898 0.2>]}
         RedCurve {keys [{arriveTangent=1.5484, interpMode=RCIM_Cubic, time=0, value=5}, <at 0.1898 0.2>]}
        }}))
     Alpha=$System.Timeline.NormalizedTime
    )
    ScaleA=true
    ScaleRGB=true
   ) {module=ScaleColor001}
   /SetVariables(Particles.ColorSprite:LinearColor=&MakeLinearColorFromVectorAndFloat(
     "Vector (RGB)"=&MakeVectorFromLinearColorRGB(Color=&ColorFromCurve(ColorCurve=DataInterface() {propertyValues {
         AlphaCurve {keys []}
         BlueCurve {keys [<at 0.0443 14>,<at 0.4563 1 RCIM_Cubic>]}
         GreenCurve {keys [<at 0.0443 14>,<at 0.4563 1 RCIM_Cubic>]}
         RedCurve {keys [<at 0.0443 14>,<at 0.4563 1 RCIM_Cubic>]}
        }}))
     "Float (Alpha)"=&FloatFromCurve(
      FloatCurve=DataInterface() {propertyValues {Curve {keys [
          {arriveTangent=31.1377, interpMode=RCIM_Cubic, leaveTangent=31.1377, tangentMode=RCTM_User, time=0, value=0}
          {arriveTangent=-2.0108, interpMode=RCIM_Cubic, leaveTangent=-2.0108, tangentMode=RCTM_User, time=0.05, value=1}
          <at 0.4 0 RCIM_Cubic>
         ]}}}
      Scale Curve=&Lerp_Float(A=0.4, B=1, Alpha=$System.Timeline.NormalizedTime)
     )
    )) {isSetParameters=true, module=SetVariables_#7}
   &DynamicMaterialParameters(
    Index 0 Param 2=&FloatFromCurve(
     FloatCurve=DataInterface() {propertyValues {Curve {keys [
         {arriveTangent=0.1337, interpMode=RCIM_Cubic, leaveTangent=0.1337, tangentMode=RCTM_User, time=0.3, value=0}
         {arriveTangent=2.5592, interpMode=RCIM_Cubic, leaveTangent=2.5592, tangentMode=RCTM_User, time=0.9, value=0.5}
        ]}}}
     Scale Curve=&Lerp_Float(A=0.1, B=0.2, Alpha=$System.Timeline.NormalizedTime)
    )
    Index 0 Param 3=&FloatFromCurve(
     FloatCurve=DataInterface() {propertyValues {Curve {keys [
         {arriveTangent=0.1034, interpMode=RCIM_Cubic, leaveTangent=0.1034, tangentMode=RCTM_User, time=0.9, value=0}
         {arriveTangent=1.3775, interpMode=RCIM_Cubic, leaveTangent=1.3775, tangentMode=RCTM_User, time=1, value=1}
        ]}}}
     Scale Curve=0.3
    )
    Index 0 Param 4=1
    Module.Param0WriteEnabled=false
    Module.Param1WriteEnabled=true
    Module.Param2WriteEnabled=true
    Module.Param3WriteEnabled=false
   ) {module=DynamicMaterialParameters001}
  ]
  rendererClasses [
   &NiagaraMeshRendererProperties
   &NiagaraSpriteRendererProperties
   &NiagaraLightRendererProperties
  ]
  renderers [
   {rendererClass=&NiagaraMeshRendererProperties, rendererIndex=0}
   {rendererClass=&NiagaraSpriteRendererProperties, rendererIndex=1}
   {rendererClass=&NiagaraLightRendererProperties, rendererIndex=2}
   {rendererClass=&NiagaraMeshRendererProperties, rendererIndex=3}
   {rendererClass=&NiagaraMeshRendererProperties, rendererIndex=4}
  ]
 }
 Lightning_Right {
  emitterUpdateScript [
   &EmitterState(Life Cycle Mode=::NewEnumerator1, Loop Duration=1, UseLoopDelay=false)
   &SpawnRate(
    Spawn Probability=&Lerp_Float(A=0.15, B=0.8, Alpha=$System.Timeline.NormalizedTime)
    SpawnRate=10
    Use Spawn Probability=true
   )
  ]
  particleSpawnScript [
   &InitializeParticle(
    Color=&RandomRangeLinearColor(Minimum=<rgba 0.7 0.7 0.7 1>, Maximum=<rgba 1.3 1.3 1.3 1>)
    Lifetime=&Lerp_Float(
     A=&RandomRangeFloat(Minimum=0.3, Maximum=0.4)
     B=&RandomRangeFloat(Minimum=0.16, Maximum=0.3)
     Alpha=$System.Timeline.NormalizedTime
    )
    Mesh Scale=&Lerp_Vector(
     A=&RandomRangeVector(Minimum=<0.8 0.8 0.8>, Maximum=<1.5 1.5 1.3>)
     B=&RandomRangeVector(Minimum=<1 1 1>, Maximum=<2 2 2>)
     Alpha=$System.Timeline.NormalizedTime
    )
    Mesh Scale Mode=::NewEnumerator4
    Sprite Rotation Mode=::NewEnumerator2
    Sprite Size Mode=::NewEnumerator3
    Uniform Sprite Size=15
    UsePositionOffset=false
    Write Material Random=true
   )
   &CameraOffset(Camera Offset Amount=5)
   /SetVariables(Particles.OffsetMesh=4) {isSetParameters=true, module=SetVariables_#8}
   /SetVariables() {isSetParameters=true, module=SetVariables_#3}
   /SetVariables(Particles.VisibilityTag=&RandomRangeInt(Minimum=0, Maximum=2)) {isSetParameters=true}
   /SetVariables() {isSetParameters=true, module=SetVariables_#4}
   &DynamicMaterialParameters(
    Index 0 Param 1=&FN_FrontEndFloatModifier(Gameplay=1, Front End=0)
    Index 0 Param 4=&SelectFloatFromArray(
     Direct Array Index=$Particles.VisibilityTag
     Direct Array Mode=::NewEnumerator1
     Float Selection Array=DataInterface() {propertyValues {FloatData [0.5,1.5,1.5]}}
    )
    Module.Param0WriteEnabled=true
    Module.Param1WriteEnabled=false
    Module.Param2WriteEnabled=false
    Module.Param3WriteEnabled=true
   )
  ]
  particleUpdateScript [
   &ParticleState()
   &SkeletalMeshLocation(
    Bone / Socket Index=&MakeCustomIntFromBool(
     Bool=$User.bisHandHidden
     True INT=&SelectIntFromArray(
      Array Sampling Mode=::NewEnumerator1
      Direct Array Index=?
      Int Selection Array=DataInterface() {propertyValues {IntData [0,4]}}
      Random Seed=0
     )
     False INT=&SelectIntFromArray(
      Array Sampling Mode=::NewEnumerator1
      Direct Array Index=?
      Int Selection Array=DataInterface() {propertyValues {IntData [0,2,4]}}
      Random Seed=0
     )
    )
    MaskSpawnGroup=false
    Mesh Sampling Type=::NewEnumerator5
    Orientation Sampling=::NewEnumerator1
    Skeletal Mesh=$System.Skeletal Mesh
    Socket Sampling Mode=::NewEnumerator5
   )
   &ScaleSpriteSize(Uniform Curve Sprite Scale=DataInterface() {propertyValues {Curve {keys [<at 0 0.3 RCIM_Cubic>,<at 0.2 1 RCIM_Cubic>]}}})
   &ScaleMeshSize(Scale Factor=&MakeVector(
     X=&FloatFromCurve(FloatCurve=DataInterface() {propertyValues {Curve {keys [
          {arriveTangent=1.0254, interpMode=RCIM_Cubic, leaveTangent=1.0254, tangentMode=RCTM_User, time=0, value=0.7}
          {interpMode=RCIM_Cubic, tangentMode=RCTM_User, time=0.5, value=1}
         ]}}})
     Y=&FloatFromCurve(FloatCurve=DataInterface() {propertyValues {Curve {keys [
          {arriveTangent=1.0254, interpMode=RCIM_Cubic, leaveTangent=1.0254, tangentMode=RCTM_User, time=0, value=0.7}
          {interpMode=RCIM_Cubic, tangentMode=RCTM_User, time=0.5, value=1}
         ]}}})
     Z=&FloatFromCurve(FloatCurve=DataInterface() {propertyValues {Curve {keys [
          {arriveTangent=0.8974, interpMode=RCIM_Cubic, leaveTangent=0.8974, tangentMode=RCTM_User, time=0, value=0.5}
          {interpMode=RCIM_Cubic, tangentMode=RCTM_User, time=0.75, value=1}
         ]}}})
    ))
   &UpdateMeshOrientation(
    Facing Direction=&RotateVectorByQuaternion(
     Vector To Rotate=&RandomRangeVector(Minimum=<0.8 0.8 0.35>, Maximum=<1 1 0.5>)
     Quaternion=$Particles.SkeletalMeshLocation.SampledRotation
    )
    Facing Interpolation=false
    Orientation Method=::NewEnumerator1
    Override Quat=false
    TransformFacing=false
    TransformUp=false
   ) {module=UpdateMeshOrientation001}
   &ScaleColor(
    Scale Alpha=&FloatFromCurve(FloatCurve=DataInterface() {propertyValues {Curve {keys [
         {arriveTangent=14.472, interpMode=RCIM_Cubic, leaveTangent=14.472, tangentMode=RCTM_User, time=0.12, value=0.5}
         {arriveTangent=-0.0363, interpMode=RCIM_Cubic, leaveTangent=-0.0363, tangentMode=RCTM_User, time=0.17, value=1}
         <at 0.9 1>
         <at 0.95 0>
        ]}}})
    Scale RGB=&Lerp_Vector(
     A=&MakeVectorFromLinearColorRGB(Color=&ColorFromCurve(ColorCurve=DataInterface() {propertyValues {
         AlphaCurve {keys []}
         BlueCurve {keys [<at 0.1007 6 RCIM_Cubic>,<at 0.1964 0.1>,<at 0.4217 0.2>,<at 0.6204 2>]}
         GreenCurve {keys [<at 0.1007 6 RCIM_Cubic>,<at 0.1964 0.1>,<at 0.4217 0.2>,<at 0.6204 2>]}
         RedCurve {keys [<at 0.1007 6 RCIM_Cubic>,<at 0.1964 0.1>,<at 0.4217 0.2>,<at 0.6204 2>]}
        }}))
     B=&MakeVectorFromLinearColorRGB(Color=&ColorFromCurve(ColorCurve=DataInterface() {propertyValues {
         AlphaCurve {keys []}
         BlueCurve {keys [{arriveTangent=1.5484, interpMode=RCIM_Cubic, time=0.0816, value=14}, <at 0.1893 0.2>, <at 0.3809 0.3>, <at 0.5904 2>]}
         GreenCurve {keys [{arriveTangent=1.5484, interpMode=RCIM_Cubic, time=0.0816, value=14}, <at 0.1893 0.2>, <at 0.3809 0.3>, <at 0.5904 2>]}
         RedCurve {keys [{arriveTangent=1.5484, interpMode=RCIM_Cubic, time=0.0816, value=14}, <at 0.1893 0.2>, <at 0.3809 0.3>, <at 0.5904 2>]}
        }}))
     Alpha=$System.Timeline.NormalizedTime
    )
    ScaleA=true
    ScaleRGB=true
   ) {enabled=false}
   &ScaleColor(
    Scale Alpha=&FloatFromCurve(
     FloatCurve=DataInterface() {propertyValues {Curve {keys [
         {arriveTangent=14.472, interpMode=RCIM_Cubic, leaveTangent=14.472, tangentMode=RCTM_User, time=0.12, value=0.5}
         {arriveTangent=-0.0363, interpMode=RCIM_Cubic, leaveTangent=-0.0363, tangentMode=RCTM_User, time=0.17, value=1}
         <at 0.95 1>
         <at 1 0>
        ]}}}
     Scale Curve=3
    )
    Scale RGB=&Lerp_Vector(
     A=&MakeVectorFromLinearColorRGB(Color=&ColorFromCurve(ColorCurve=DataInterface() {propertyValues {
         AlphaCurve {keys []}
         BlueCurve {keys [<at 0.0146 1 RCIM_Cubic>,<at 0.1743 0.07>]}
         GreenCurve {keys [<at 0.0146 1 RCIM_Cubic>,<at 0.1743 0.07>]}
         RedCurve {keys [<at 0.0146 1 RCIM_Cubic>,<at 0.1743 0.07>]}
        }}))
     B=&MakeVectorFromLinearColorRGB(Color=&ColorFromCurve(ColorCurve=DataInterface() {propertyValues {
         AlphaCurve {keys []}
         BlueCurve {keys [{arriveTangent=1.5484, interpMode=RCIM_Cubic, time=0, value=5}, <at 0.1898 0.2>]}
         GreenCurve {keys [{arriveTangent=1.5484, interpMode=RCIM_Cubic, time=0, value=5}, <at 0.1898 0.2>]}
         RedCurve {keys [{arriveTangent=1.5484, interpMode=RCIM_Cubic, time=0, value=5}, <at 0.1898 0.2>]}
        }}))
     Alpha=$System.Timeline.NormalizedTime
    )
    ScaleA=true
    ScaleRGB=true
   ) {module=ScaleColor001}
   /SetVariables(Particles.ColorSprite:LinearColor=&MakeLinearColorFromVectorAndFloat(
     "Vector (RGB)"=&MakeVectorFromLinearColorRGB(Color=&ColorFromCurve(ColorCurve=DataInterface() {propertyValues {
         AlphaCurve {keys []}
         BlueCurve {keys [<at 0.0443 14>,<at 0.4563 1 RCIM_Cubic>]}
         GreenCurve {keys [<at 0.0443 14>,<at 0.4563 1 RCIM_Cubic>]}
         RedCurve {keys [<at 0.0443 14>,<at 0.4563 1 RCIM_Cubic>]}
        }}))
     "Float (Alpha)"=&FloatFromCurve(
      FloatCurve=DataInterface() {propertyValues {Curve {keys [
          {arriveTangent=31.1377, interpMode=RCIM_Cubic, leaveTangent=31.1377, tangentMode=RCTM_User, time=0, value=0}
          {arriveTangent=-2.0108, interpMode=RCIM_Cubic, leaveTangent=-2.0108, tangentMode=RCTM_User, time=0.05, value=1}
          <at 0.4 0 RCIM_Cubic>
         ]}}}
      Scale Curve=&Lerp_Float(A=0.4, B=1, Alpha=$System.Timeline.NormalizedTime)
     )
    )) {isSetParameters=true, module=SetVariables_#7}
   &DynamicMaterialParameters(
    Index 0 Param 2=&FloatFromCurve(
     FloatCurve=DataInterface() {propertyValues {Curve {keys [
         {arriveTangent=0.1337, interpMode=RCIM_Cubic, leaveTangent=0.1337, tangentMode=RCTM_User, time=0.3, value=0}
         {arriveTangent=2.5592, interpMode=RCIM_Cubic, leaveTangent=2.5592, tangentMode=RCTM_User, time=0.9, value=0.5}
        ]}}}
     Scale Curve=&Lerp_Float(A=0.1, B=0.2, Alpha=$System.Timeline.NormalizedTime)
    )
    Index 0 Param 3=&FloatFromCurve(
     FloatCurve=DataInterface() {propertyValues {Curve {keys [
         {arriveTangent=0.1034, interpMode=RCIM_Cubic, leaveTangent=0.1034, tangentMode=RCTM_User, time=0.9, value=0}
         {arriveTangent=1.3775, interpMode=RCIM_Cubic, leaveTangent=1.3775, tangentMode=RCTM_User, time=1, value=1}
        ]}}}
     Scale Curve=0.3
    )
    Index 0 Param 4=1
    Module.Param0WriteEnabled=false
    Module.Param1WriteEnabled=true
    Module.Param2WriteEnabled=true
    Module.Param3WriteEnabled=false
   ) {module=DynamicMaterialParameters001}
  ]
  rendererClasses [
   &NiagaraSpriteRendererProperties
   &NiagaraMeshRendererProperties
   &NiagaraLightRendererProperties
  ]
  renderers [
   {rendererClass=&NiagaraSpriteRendererProperties}
   {rendererClass=&NiagaraMeshRendererProperties, rendererIndex=1}
   {rendererClass=&NiagaraLightRendererProperties, rendererIndex=2}
   {rendererClass=&NiagaraMeshRendererProperties, rendererIndex=3}
   {rendererClass=&NiagaraMeshRendererProperties, rendererIndex=4}
  ]
 }
}
summary {
 systemName=NS_Titanium_Character_Idle
 emitters [
  {
   emitterName=Lightning_Right
   rendererClasses [
    &NiagaraSpriteRendererProperties
    &NiagaraMeshRendererProperties
    &NiagaraLightRendererProperties
   ]
  }
  {
   emitterName=Lightning_Left
   rendererClasses [
    &NiagaraMeshRendererProperties
    &NiagaraSpriteRendererProperties
    &NiagaraLightRendererProperties
   ]
  }
  {
   emitterName=Lightning_Burst_Left
   rendererClasses [
    &NiagaraSpriteRendererProperties
    &NiagaraLightRendererProperties
    &NiagaraMeshRendererProperties
   ]
  }
  {
   emitterName=Lightning_Burst_Right
   rendererClasses [
    &NiagaraSpriteRendererProperties
    &NiagaraLightRendererProperties
    &NiagaraMeshRendererProperties
   ]
  }
 ]
 userVariables [
  User.bIsFrontEnd {type=NiagaraBool}
  User.bIsFrontEndPreview {type=NiagaraBool, defaultValue=true}
  User.bisHandHidden {type=NiagaraBool}
  User.Empowered {type=NiagaraBool}
 ]
}
